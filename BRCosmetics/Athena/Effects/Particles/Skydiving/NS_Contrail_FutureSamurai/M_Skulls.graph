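# Translucent unlit Niagara particle material that draws skulls for the FutureSamurai skydiving contrail.
# Particle-space position picks a cell in 2x2 SubUV skull textures, warped by a polar-style scrolled water normal except at low quality.
# Dynamic parameters threshold color, opacity and erosion masks, with camera depth fade feeding the erosion.
# The color mask blends time-of-day tint, shaded by normal against sun vector, into particle color.
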Material /BRCosmetics/Athena/Effects/Particles/Skydiving/NS_Contrail_FutureSamurai/M_Skulls
graph {
 properties {
  bUsedWithNiagaraMeshParticles=true
  bUsedWithNiagaraSprites=true
  blendMode=BLEND_Translucent
  pixelDepthOffsetMode=PDOM_Legacy
  shadingModel=MSM_Unlit
  usedShadingModels=Unlit
 }
 nodes [
  n1 = CollectionParameter() {
   collection=&FortniteMaterialParameters
   parameterName=SunlightDirectionalVector
  }
  n2 = MaterialFunctionCall(Float3=@n1) {materialFunction=&BreakOutFloat3Components}
  n3 = DotProduct(
   A=VertexNormalWS()
   B=MaterialFunctionCall(R=@n2.R, G=@n2.R, B=@n2.R) {materialFunction=&AppendMany}.RGB
  )
  n4 = Multiply(
   A=MaterialFunctionCall(
    Color_Morning=VectorParameter() {parameterName=Morning, defaultValue=<rgba 0.0117 0.0154 0.026 1>}.RGB
    Color_Day=VectorParameter() {parameterName=Day, defaultValue=<rgba 0.002 0.0173 0.0208 1>}.RGB
    Color_Evening=VectorParameter() {parameterName=Evening, defaultValue=<rgba 0 0 0 1>}.RGB
    Color_Night=VectorParameter() {parameterName=Night, defaultValue=<rgba 0.0134 0.0159 0.0208 1>}.RGB
   ) {materialFunction=&Color_Time_of_Day}.Result
   B=@n3
  )
  n5 = ParticleColor()
  n6 = TransformPosition(None=WorldPosition().XYZ) {transformSourceType=TRANSFORMPOSSOURCE_World, transformType=TRANSFORMPOSSOURCE_Particle}
  n7 = Constant3Vector() {constant=<rgba 0 0 -50 0>}
  n8 = MaterialFunctionCall(Float3=Multiply(A=Add(A=@n6, B=@n7)) {constB=0.005, desc=scale}) {materialFunction=&BreakOutFloat3Components}
  n9 = Divide(
   A=Arctangent2Fast(Y=@n8.B, X=@n8.B)
   B=MaterialFunctionCall(Multiplier=Constant() {r=2}) {materialFunction=&Pi}.Result
  ) {constB=6.28}
  n10 = MaterialFunctionCall(Vector 2=AppendVector(A=@n8.G, B=@n8.G)) {materialFunction=/Engine/Functions/Engine_MaterialFunctions02/Utility/VectorLength}
  n11 = DynamicParameter() {parameterIndex=1, defaultValue=<rgba 0.0381 0 0 1>, paramNames [Color Erosion, Opacity Erosion, FE Erode, Distortion Offset]}
  n12 = DynamicParameter() {defaultValue=<rgba 0.0381 0 1 1>, paramNames [Color Erosion, Opacity Erosion, Sub UV, Distortion Offset]}
  n13 = Add(A=Multiply(A=SquareRoot(None=@n10.V2 Length), B=@n11.Opacity Erosion), B=@n12.Distortion Offset)
  n14 = TextureSample!(UVs=AppendVector(A=@n9, B=@n13), Tex, "DDX(UVs)"=DDX(Value=@n9), "DDY(UVs)"=DDY(Value=@n13), Apply View MipBias) {
   mipValueMode=TMVM_Derivative
   samplerSource=SSM_Wrap_WorldGroupSettings
   samplerType=SAMPLERTYPE_Normal
   texture=/Game/Packages/Fortnite_UniqueMaterials_1/Textures/T_Water_TilingNormal_Waves_01
  }
  n15 = AppendVector(A=@n14.R, B=@n14.R)
  n16 = MaterialFunctionCall(Float3=Multiply(A=@n6) {constB=0.0015, desc=scale}) {materialFunction=&BreakOutFloat3Components}
  n17 = OneMinus(None=@n16.B)
  n18 = DynamicParameter() {defaultValue=<rgba 0.0381 0 1 1>, paramNames [Color Erosion, Opacity Erosion, Sub UV, Distortion Offset]}
  n19 = Constant() {r=2}
  n20 = Fmod(A=@n18.Sub UV, B=@n19)
  n21 = Divide(A=@n18.Sub UV, B=@n19)
  n22 = Add(A=Add(A=AppendVector(A=@n16.G, B=@n17), B=Constant2Vector() {g=0.25, r=0.25}), B=AppendVector(A=Divide(A=@n20, B=@n19), B=Floor(None=@n21)))
  n23 = QualitySwitch(Default=Add(A=Multiply(A=@n15, B=@n11.Color Erosion), B=@n22), Low=@n22)
  n24 = TextureSample!(UVs=@n23, Tex, Apply View MipBias) {
   samplerType=SAMPLERTYPE_LinearColor
   texture=&/T_Contrail_FutureSamurai_Skulls_SubUV_Color
  }
  n25 = DynamicParameter() {defaultValue=<rgba 0.0381 0 1 1>, paramNames [Color Erosion, Opacity Erosion, Sub UV, Distortion Offset]}
  n26 = Saturate(None=Multiply(A=Subtract(A=@n24.RGB, B=@n25.Color Erosion)) {constB=10})
  MaterialFunctionCall(Emissive=LinearInterpolate(A=@n4, B=@n5.RGB, Alpha=@n26)) {materialFunction=&MF_FX_EmissiveAdjustments}
  n27 = MaterialFunctionCall(Fade Length=Constant() {r=200}) {materialFunction=&CameraDepthFade}
  n28 = DynamicParameter() {parameterIndex=1, defaultValue=<rgba 0.0381 0 0 1>, paramNames [Color Erosion, Opacity Erosion, FE Erode, Distortion Offset]}
  Max(A=@n27.Result, B=@n28.FE Erode)
  Multiply(A=@n5.RGB, B=Add(A=@n4, B=@n26))
  n29 = TextureSample(UVs=@n23) {
   samplerType=SAMPLERTYPE_LinearColor
   texture=&/T_Contrail_FutureSamurai_Skulls_SubUV_Opacity
  }
  n30 = TextureSample(UVs=@n23) {samplerType=SAMPLERTYPE_LinearGrayscale, texture=/BRCosmetics/Effects/Fort_Effects/BluGlow/Violentium/Cellular_basecolor}
  n31 = Multiply(A=@n29.R, B=@n30.R)
  n32 = DynamicParameter() {defaultValue=<rgba 0.0381 0 1 1>, paramNames [Color Erosion, Opacity Erosion, Sub UV, Distortion Offset]}
  n33 = Multiply(A=Subtract(A=Add(A=@n29.R, B=@n31), B=@n32.Opacity Erosion)) {constB=10}
  n34 = TextureSample(UVs=@n23) {texture=&/T_Contrail_FutureSamurai_Skulls_SubUV_Erosion}
  n35 = Min(A=@n27.Result, B=@n28.FE Erode)
  n36 = Saturate(None=Multiply(A=Subtract(A=@n34.R, B=@n35)) {constB=10})
  Multiply(A=ParticleColor().A, B=Multiply(A=Saturate(None=@n33), B=OneMinus(None=@n36)))
  Multiply(A=Multiply(A=@n3, B=@n30.R), B=OneMinus(None=@n26))
  Power()
  SkyAtmosphereLightDirection()
 ]
}
