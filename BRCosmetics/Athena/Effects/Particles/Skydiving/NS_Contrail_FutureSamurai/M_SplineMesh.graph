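# Translucent, two-sided Niagara mesh-particle material, apparently for the FutureSamurai skydiving contrail, drawing a soft smoke-trail strip.
# An opacity chain multiplies a swizzled, dynamically tiled and offset smoke mask by centred V falloff, particle alpha and camera depth fade.
# A particle-space offset scales a cross-product axis by U and dynamic SplineWidth.
# The camera-facing blend has identical inputs, and the pan-speed Panner does not feed the mask.
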Material /BRCosmetics/Athena/Effects/Particles/Skydiving/NS_Contrail_FutureSamurai/M_SplineMesh
graph {
 properties {
  bUsedWithNiagaraMeshParticles=true
  blendMode=BLEND_Translucent
  pixelDepthOffsetMode=PDOM_Legacy
  twoSided=true
  usedShadingModels=Default Lit
 }
 nodes [
  n1 = Transform(None=Constant3Vector() {constant=<rgba 0 1 0 0>}) {transformType=TRANSFORM_ParticleWorld}
  n2 = CrossProduct(A=Constant3Vector() {constant=<rgba 0 0 1 0>}, B=@n1)
  CrossProduct(A=@n2, B=@n1)
  n3 = DynamicParameter() {parameterIndex=1, paramNames [Opacity Texture Offset, Opacity Texture PanSpeed, Opacity Texture Tiling, Random]}
  n4 = MaterialFunctionCall(Float2=MaterialFunctionCall(XY=TextureCoordinate() {vTiling=0.5}) {materialFunction=&Swizzle}.YX) {materialFunction=&BreakOutFloat2Components}
  n5 = AppendVector(A=Add(A=@n3.Opacity Texture Offset, B=Multiply(A=@n3.Opacity Texture Tiling, B=@n4.R)), B=@n4.G)
  n6 = TextureSampleParameter2D(UVs=@n5) {
   parameterName=Mask
   samplerSource=SSM_Wrap_WorldGroupSettings
   samplerType=SAMPLERTYPE_Grayscale
   texture=&T_Smoke_Trail_Soft
  }
  n7 = Subtract(A=ComponentMask(None=TextureCoordinate()) {g=true}, B=Constant() {r=0.5}) {constB=0.5}
  n8 = Constant() {r=2}
  n9 = OneMinus(None=Abs(None=Multiply(A=@n7, B=@n8) {constB=2}))
  Multiply(
   A=Multiply(A=ParticleColor().A, B=Multiply(A=@n6.R, B=@n9))
   B=MaterialFunctionCall(Fade Length=ScalarParameter() {defaultValue=80, parameterName=Camera Fade distance}) {materialFunction=&CameraDepthFade}.Result
  )
  Panner(
   Coordinate=@n5
   Speed=DynamicParameter() {parameterIndex=1, paramNames [Opacity Texture Offset, Opacity Texture PanSpeed, Opacity Texture Tiling, Random]}.Opacity Texture PanSpeed
  )
  n10 = Constant3Vector() {constant=<rgba 0 0 1 0>}
  n11 = Abs(None=DotProduct(
    A=Transform(None=@n10) {transformSourceType=TRANSFORMSOURCE_ParticleWorld}
    B=MaterialFunctionCall() {materialFunction=&CameraDirectionVector}.Result
   ))
  n12 = LinearInterpolate(A=@n2, B=@n2, Alpha=OneMinus(None=Multiply(A=@n11)))
  n13 = Multiply(
   A=Constant() {r=10}
   B=MaterialFunctionCall(Float2=TextureCoordinate()) {materialFunction=&BreakOutFloat2Components}.R
  )
  Transform(None=Multiply(A=Multiply(A=@n12, B=@n13), B=DynamicParameter() {paramNames [Erode,SplineWidth,PanAdd,Random]}.SplineWidth)) {transformSourceType=TRANSFORMSOURCE_ParticleWorld}
 ]
}
