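# Niagara skydiving contrail for the FutureSamurai cosmetic, trailing red and black smoke, skulls and meshes from the character's skeletal mesh.
# Layered smoke ribbons, fewer at lower quality, join bone-attached smoke sprites and GPU meshes aligned to velocity.
# Camera-oriented skull meshes carry smoke sprites and feed an attached ribbon emitter through particle reads.
# Dive state and owner velocity drive gameplay motion and spawning, while front-end previews use paraglide lean instead.
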NiagaraSystem /BRCosmetics/Athena/Effects/Particles/Skydiving/NS_Contrail_FutureSamurai/NS_Contrail_FutureSamurai_02
dependencies {
 usedDataInterfaces [
  &NiagaraDataInterfaceCurve
  &NiagaraDataInterfaceVectorCurve
  &NiagaraDataInterfaceCamera
  &NiagaraDataInterfaceParticleRead
  &NiagaraDataInterfaceVector2DCurve
 ]
 usedDynamicInputs [&Add_Float, &SetBoolByFloatComparison]
 usedModules [
  &SystemState
  /Game/Effects/Niagara/Modules/Skydiving/FN_DivingAttributes/v4/FN_DivingAttributes_v4
  &self:SystemSpawnScript.SystemScriptSource.SystemScriptGraph.NiagaraNodeAssignment_0.SetVariables_#1
 ]
}
emitters {
 FN_AttachedRibbon_v3001 {
  emitterUpdateScript [
   &EmitterState()
   /Game/Effects/Niagara/Modules/Attributes/FN_Source_v2(AttributeReader=DataInterface() {propertyValues {EmitterBinding {bindingMode=Other, emitterName=Skulls}}})
   &SpawnRate(
    Spawn Probability=&FN_FrontEndFloatModifier(Gameplay=0, Front End=$User.ParaGlideLeanAlpha)
    SpawnRate=100
    Use Spawn Probability=true
   )
  ]
  particleSpawnScript [
   /Game/Effects/Niagara/Modules/Attributes/FN_SourceParentID(Ribbon ID from Parent ID=true)
   /Game/Effects/Niagara/Modules/Ribbons/FN_InitializeRibbon_v5(
    Custom Offset=true
    Lifetime:NiagaraFloat=/Niagara/DynamicInputs/Math/Min_Float(
     A:NiagaraFloat=3
     B:NiagaraFloat=/Game/Effects/Niagara/DynamicInputs/Attributes/FN_GetFloatByID(Attribute:ENiagaraCommonFloats=::NewEnumerator0, Default on Invalid:NiagaraFloat=0)
    )
    Local Vector Offset:Vector3f=/Game/Effects/Niagara/DynamicInputs/Attributes/FN_GetVectorByID(Attrib name:ENiagaraCommonVectors=::NewEnumerator1, Default on Invalid001:Vector3f=<0 0 0>)
    Offset Amount=0.5
    Position:Vector3f=&FN_GetVectorByID(Attrib name:ENiagaraCommonVectors=::NewEnumerator0, Default on Invalid001:Vector3f=<0 0 0>)
    Ribbon Width:NiagaraFloat=&FN_FrontEndFloatModifier(Gameplay=40, Front End=20)
   )
   /SetVariables(Particles.RibbonID=$Particles.ParentID) {isSetParameters=true}
   &AddVelocity(Velocity=&FN_FrontEndVector3Modifier(
     Gameplay=<0 0 -100>
     Front End=&RotateVectorByQuaternion(
      Vector To Rotate=&MakeVector(X=0, Y=0, Z=-200)
      Quaternion=&MakeQuatFromAxisAngle(Axis=<1 0 0>, AngleInDegrees=&Multiply_Float(A=$User.ParaGlideLeanAlpha, B=63.4543))
     )
    ))
  ]
  particleUpdateScript [
   &ParticleState()
   &SolveForcesAndVelocity(Clamp Velocity=false, Limit Acceleration=false)
   /Game/Effects/Niagara/Modules/Materials/FN_RibbonMaterial/FN_RibbonMaterial_v2(
    Color:LinearColor=&MakeLinearColor(
     R:NiagaraFloat=0
     G:NiagaraFloat=0
     B:NiagaraFloat=0
     A:NiagaraFloat=&FloatFromCurve(
      FloatCurve=DataInterface() {propertyValues {Curve {keys [
          <at -0.0251 -0.0463 RCIM_Cubic>
          {arriveTangent=1.1831, interpMode=RCIM_Cubic, leaveTangent=1.1831, time=0.3693, value=1}
          <at 0.8258 0.9604>
          <at 0.9906 -0.012 RCIM_Cubic>
         ]}}}
      CurveIndex=&FN_GetFloatByID(Attribute:ENiagaraCommonFloats=::NewEnumerator1, Default on Invalid:NiagaraFloat=1)
      Scale Curve=&FloatFromCurve(FloatCurve=DataInterface() {propertyValues {Curve {keys [<at -0.0095 -0.0206>,<at 0.0987 1.0063>,<at 0.7674 0.0032>]}}}, Scale Curve=1.5)
     )
    )
    Opacity Erosion=0.4186
    Opacity Pan U Offset:NiagaraFloat=&RandomRangeFloat(Minimum=0, Maximum=1, Randomness Mode=::NewEnumerator1, Random Seed=$Particles.SourceID, Fixed Random Seed=true)
    Opacity Pan U Speed:NiagaraFloat=&FN_FrontEndFloatModifier(Gameplay=-1, Front End=0.25)
    Opacity Power:NiagaraFloat=&RandomRangeFloat(Minimum=0.2, Maximum=0.5, Randomness Mode=::NewEnumerator1, Random Seed=$Particles.SourceID, Fixed Random Seed=true)
    Opacity Scale U=0.9341
   )
  ]
  rendererClasses [&NiagaraRibbonRendererProperties, &NiagaraSpriteRendererProperties]
  renderers [
   {rendererClass=&NiagaraRibbonRendererProperties}
   {rendererClass=&NiagaraSpriteRendererProperties, rendererIndex=1}
  ]
 }
 Hand_Smoke {
  emitterUpdateScript [
   &EmitterState()
   &SpawnRate(SpawnRate=10, Use Spawn Probability=false)
  ]
  particleSpawnScript [
   /BRCosmetics/Effects/Niagara/Modules/Groups/FN_SetSpawnGroup_v3/FN_SetSpawnGroup_v3(Number of Groups=4, Pattern:ENiagaraPartitionWaveForm=::NewEnumerator1)
   &FN_Sprite_Attributes(Lifetime:NiagaraFloat=&Lerp_Float(
     A=&RandomRangeFloat(Minimum=2, Maximum=3)
     B=&RandomRangeFloat(Minimum=0.5, Maximum=1)
     Alpha=&FN_FrontEndFloatModifier(Gameplay=$System.Dive, Front End=$User.ParaGlideLeanAlpha)
    )) {module=FN_Sprite_v3}
  ]
  particleUpdateScript [
   &ParticleState()
   &SkeletalMeshLocation(
    Bone / Socket Index=$Particles.SpawnGroup
    Bone Sampling Mode=::NewEnumerator4
    Inherit Velocity=false
    Interpolate Mesh Sampling=false
    MaskSpawnGroup=false
    Sampled Position Offset=&FN_FrontEndVector3Modifier(Gameplay=<0 0 -5>, Front End=<0 0 0>)
    Sampled Position Offset Coordinate Space:ENiagaraMeshTransforms=::NewEnumerator2
    Skeletal Mesh=$User.SkeletalMesh
   )
   &FN_Sprite_Attributes(
    Camera Offset=81.3418
    Create Camera Offset Attribute=true
    Create Sprite Rotation Attribute=true
    Create SpriteSize Attribute=true
    SpriteRotation:NiagaraFloat=&RandomRangeFloat(Minimum=0, Maximum=1)
    SpriteSize=&Vector2DFromFloat(Value=&FloatFromCurve(
      FloatCurve=DataInterface() {propertyValues {Curve {keys [{arriveTangent=6.1188, interpMode=RCIM_Cubic, leaveTangent=6.1188, tangentMode=RCTM_User, time=0, value=1}, <at 0.9949 3.019>]}}}
      Scale Curve=&RandomRangeFloat(Minimum=10, Maximum=25)
     ))
   ) {module=FN_Sprite_v3001}
   /Game/Effects/Niagara/Modules/Sprite/FN_SpriteUV(
    End Frame:NiagaraInt32=&RandomRangeInt(Minimum=55, Maximum=63)
    Randomize Sprite UV scale=true
    Seed=2
    Start Frame:NiagaraInt32=&RandomRangeInt(Minimum=4, Maximum=9)
    SubUV Lookup Index:NiagaraFloat=$Particles.NormalizedAge
    Write to Module Namepsace=true
   )
   /SetVariables(Particles.Alpha:NiagaraFloat=&FloatFromCurve(
     FloatCurve=DataInterface() {propertyValues {Curve {keys [<at -0.0044 -0.015 RCIM_Cubic>, {arriveTangent=0.0149, interpMode=RCIM_Cubic, leaveTangent=0.0149, time=0.3249, value=0.9916}, <at 1 0>]}}}
     Scale Curve=1.4
    )) {isSetParameters=true}
   /SetVariables(
    Particles.BlackTurbulentColor:LinearColor=&MakeLinearColorFromVectorAndFloat("Vector (RGB)"=<0 0 0>, "Float (Alpha)"=$Particles.Alpha)
    Particles.Smoke Color:LinearColor=&MakeLinearColorFromVectorAndFloat("Vector (RGB)"=<0.201 0 0>, "Float (Alpha)"=$Particles.Alpha)
   ) {isSetParameters=true}
  ]
  rendererClasses [&NiagaraSpriteRendererProperties]
  renderers [
   {rendererClass=&NiagaraSpriteRendererProperties}
   {rendererClass=&NiagaraSpriteRendererProperties, rendererIndex=1}
  ]
 }
 Skulls {
  emitterUpdateScript [
   &EmitterState()
   /SetVariables(Emitter.Camera Pos:Vector3f=&ReturnCameraVector(Camera Property=::NewEnumerator1, Coordinate Space=::Simulation)) {isSetParameters=true}
   /SetVariables() {enabled=false, isSetParameters=true}
   &SpawnRate(
    Spawn Probability=&FN_FrontEndFloatModifier(Gameplay=$System.Dive, Front End=$User.ParaGlideLeanAlpha)
    SpawnRate=&FN_FrontEndFloatModifier(
     Gameplay=&RandomRangeFloat(Minimum=5, Maximum=10, Evaluation Type=::NewEnumerator1)
     Front End=&RandomRangeFloat(Minimum=2, Maximum=5, Evaluation Type=::NewEnumerator1)
    )
    Use Spawn Probability=true
   )
   &SpawnRate(Use Spawn Probability=false) {enabled=false, module=SpawnRate001}
   &SpawnBurst_Instantaneous(Spawn Count=1, Use Spawn Probability=false) {enabled=false}
  ]
  particleSpawnScript [
   /Niagara/Modules/Spawn/Initialization/InitializeParticle(
    Color=<rgba 1 0 0 1>
    Lifetime=&RandomRangeFloat(Minimum=1.5, Maximum=2.5)
    Mesh Scale=&VectorFromFloat(Value=&RandomRangeFloat(Minimum=0.04, Maximum=0.06))
    Module.Write Color=true
    Module.Write Lifetime=true
    Module.Write Mass=true
    Module.Write Material Random=true
    Module.Write Position=false
    Module.Write Scale=true
    Module.Write SpriteRotation=true
    Module.Write SpriteSize=true
    Sprite Rotation=&RandomRangeFloat(Minimum=0, Maximum=360)
    Sprite Size=<100 100>
   )
   &FN_SetSpawnGroup_v3(Number of Groups=2, Pattern:ENiagaraPartitionWaveForm=::NewEnumerator1)
   &self:Spawn_Group_Prototype(Cycle Spawn Group Value=4) {enabled=false, module=Spawn_Group_Prototype}
   &SkeletalMeshLocation(
    Bone / Socket Index=$Particles.SpawnGroup
    Bone Sampling Mode=::NewEnumerator4
    Inherit Velocity=false
    Interpolate Mesh Sampling=false
    MaskSpawnGroup=false
    Skeletal Mesh=$User.SkeletalMesh
   )
   &AddVelocity(Velocity=<0 0 100>) {enabled=false, module=AddVelocity001}
   &AddVelocity(
    Scale Added Velocity=<0.5 0.5 0.5>
    Velocity=&FN_FrontEndVector3Modifier(
     Gameplay=&Add_Vector(
      A=&TransformVector(
       Vector=&Multiply_VectorByFloat(
        Vector=&NormalizeVector(Vector To Normalize=$Engine.Owner.Velocity)
        Float=&Multiply_Float(A=400, B=HlslExpression() {hlslExpression="sign(dot(Engine.Owner.Velocity, float3(0,0,1)))"})
       )
       Source Space=::World
       Destination Space=::Local
      )
      B=&self:ScratchDynamicInput(
       Multiplier:NiagaraFloat=&Lerp_Float(
        A=0
        B=/Game/Effects/Niagara/DynamicInputs/Skydiving/SkydiveHandsFeetFloatModifier(Hands:NiagaraFloat=50, Feet:NiagaraFloat=25)
        Alpha=$System.Dive
       )
       Source:Vector3f=Linked() {linkedVariable {name=Particles.Position, type=NiagaraPosition}}
      )
     )
     Front End=&RotateVectorByQuaternion(
      Vector To Rotate=/BRCosmetics/Effects/Niagara/DynamicInputs/Skydiving/SkydiveHandsFeetVector3Modifier(
       Hands:Vector3f=&MakeVector(
        X=&Multiply_Float(A=$User.ParaGlideLeanAlpha, B=HlslExpression() {hlslExpression="Particles.SpawnGroup && 1 ? -40 : 40"})
        Y=0
        Z=150
       )
       Feet:Vector3f=<0 0 150>
      )
      Quaternion=&MakeQuatFromAxisAngle(Axis=<1 0 0>, AngleInDegrees=&Multiply_Float(A=$User.ParaGlideLeanAlpha, B=63.4543))
     )
    )
   )
   /Niagara/Modules/Spawn/Orientation/V2/InitialMeshOrientation(
    Orientation Vector=&RandomVector(Vector Scale=1)
    Rotation=&MakeVector(
     X=0
     Y=&RandomRangeFloat(Minimum=-0.1, Maximum=0.1)
     Z=&RandomRangeFloat(Minimum=-0.1, Maximum=0.1)
    )
    Use Orientation Vector=false
    Use Rotation Vector=true
   )
   &SubUVAnimation(End Frame=3, SubUV Animation Mode:ENiagara_SubUVLookupMode=::NewEnumerator2)
   &FN_SpriteUV(Randomize Sprite UV scale=true)
   /SetVariables(
    Particles.OrientationOffset:Vector3f=<-5 0 0>
    Particles.RandomOffset_01:Vector3f=&RandomVector(Vector Scale=1)
    Particles.RandomOffset_02:Vector3f=&RandomVector(Vector Scale=1)
   ) {isSetParameters=true}
  ]
  particleUpdateScript [
   &ParticleState()
   &SolveForcesAndVelocity(Clamp Velocity=false, Limit Acceleration=false)
   &Color(Color=&MakeLinearColor(
     R:NiagaraFloat=&FloatFromCurve(FloatCurve=DataInterface() {propertyValues {Curve {keys [<at 0.242 1.0111>,<at 1 0>]}}})
     G:NiagaraFloat=0
     B:NiagaraFloat=0
     A:NiagaraFloat=&FloatFromCurve(FloatCurve=DataInterface() {propertyValues {Curve {keys [<at 0.8728 1>,<at 1 0>]}}})
    ))
   &ScaleMeshSize(Scale Factor=&VectorFromFloat(Value=&FloatFromCurve(FloatCurve=DataInterface() {propertyValues {Curve {keys [{arriveTangent=1.7682, interpMode=RCIM_Cubic, leaveTangent=1.7682, tangentMode=RCTM_User, time=0.0011, value=0.9204}, <at 1.0138 1.9699>]}}})))
   /SetVariables(Particles.MeshOrientation=&MultiplyQuaternion(
     Quaternion A=&self:ScratchDynamicInput_02(
      Camera Vector:Vector3f=$Emitter.Camera Pos
      Engine Axis Vector:Vector3f=&TransformVector(
       Vector=&FN_FrontEndVector3Modifier(Gameplay=$Engine.Owner.SystemYAxis, Front End=<0 0 1>)
       Source Space=::World
       Destination Space=::Simulation
      )
     )
     Quaternion B=$Particles.Initial.MeshOrientation
    )) {isSetParameters=true, module=SetVariables_#2}
   &DynamicMaterialParameters(
    Index 0 Param 1=&FloatFromCurve(
     FloatCurve=DataInterface() {propertyValues {Curve {keys [<at 0 1 RCIM_Cubic>, {arriveTangent=-0.9127, interpMode=RCIM_Cubic, leaveTangent=-0.9127, tangentMode=RCTM_User, time=0.9959, value=-0.0166}]}}}
     Scale Curve=0.25
    )
    Index 0 Param 2=&FN_FrontEndFloatModifier(
     Gameplay=&FloatFromCurve(
      FloatCurve=DataInterface() {propertyValues {Curve {keys [<at 0.613 0.9719 RCIM_Cubic>,<at 0.9745 0>]}}}
      CurveIndex=&Multiply_Float(A=$Particles.NormalizedAge, B=2)
     )
     Front End=&FloatFromCurve(
      FloatCurve=DataInterface() {propertyValues {Curve {keys [<at 0.001 0.9748 RCIM_Cubic>,<at 0.9745 0>]}}}
      CurveIndex=&Multiply_Float(A=$Particles.NormalizedAge, B=2)
     )
    )
    Index 0 Param 3=$Particles.Initial.SubImageIndex
    Index 0 Param 4=&FloatFromCurve(
     FloatCurve=DataInterface() {propertyValues {Curve {keys [<at 0 1>,<at 0.9944 -0.0017>]}}}
     CurveIndex=&Multiply_Float(A=$Particles.NormalizedAge, B=1)
     Scale Curve=0.5
    )
    Index 1 Param 1=&FloatFromCurve(
     FloatCurve=DataInterface() {propertyValues {Curve {keys [<at -0.0041 0.172 RCIM_Cubic>,<at 1 1>]}}}
     CurveIndex=&Multiply_Float(A=$Particles.NormalizedAge, B=2)
     Scale Curve=0.0636
    )
    Index 1 Param 2=0.2
    Index 1 Param 3=&FN_FrontEndFloatModifier(
     Gameplay=1
     Front End=&FloatFromCurve(FloatCurve=DataInterface() {propertyValues {Curve {keys [<at 0.1427 1.0432 RCIM_Cubic>,<at 0.9912 0.0122>]}}})
    )
    Module.Param0WriteEnabled=true
    Module.Param0WriteEnabled1=true
    Module.Param1WriteEnabled=true
    Module.Param1WriteEnabled1=true
    Module.Param2WriteEnabled=true
    Module.Param2WriteEnabled1=true
    Module.Param3WriteEnabled=true
    Module.Param3WriteEnabled1=false
    Write Parameter Index 1=true
   )
   &FN_Sprite_Attributes(
    Camera Offset=20
    Create Camera Offset Attribute=true
    Create Position Attribute=true
    Create SpriteSize Attribute=true
    Position:Vector3f=&Add_Vector(
     A=&Add_Vector(A=Linked() {linkedVariable {name=Particles.Position, type=NiagaraPosition}}, B=$Particles.RandomOffset_01)
     B=&RotateVectorByQuaternion(Vector To Rotate=$Particles.OrientationOffset, Quaternion=$Particles.MeshOrientation)
    )
    SpriteSize=&Vector2DFromFloat(Value=&FloatFromCurve(
      FloatCurve=DataInterface() {propertyValues {Curve {keys [{arriveTangent=6.1188, interpMode=RCIM_Cubic, leaveTangent=6.1188, tangentMode=RCTM_User, time=0, value=1}, <at 0.9949 3.019>]}}}
      Scale Curve=&RandomRangeFloat(Minimum=10, Maximum=20)
     ))
    Write to Module Namespace=true
   ) {module=FN_Sprite_v3001}
   &FN_SpriteUV(
    End Frame:NiagaraInt32=&RandomRangeInt(Minimum=55, Maximum=63)
    Start Frame:NiagaraInt32=&RandomRangeInt(Minimum=30, Maximum=35)
    SubUV Lookup Index:NiagaraFloat=$Particles.NormalizedAge
    Write to Module Namepsace=true
   ) {module=FN_SpriteUV001}
   &FN_SpriteUV(
    End Frame:NiagaraInt32=&RandomRangeInt(Minimum=55, Maximum=63)
    Randomize Sprite UV scale=true
    Start Frame:NiagaraInt32=&RandomRangeInt(Minimum=20, Maximum=30)
    SubUV Lookup Index:NiagaraFloat=$Particles.NormalizedAge
    Write to Module Namepsace=true
   ) {module=FN_SpriteUV002}
   &FN_Sprite_Attributes(
    Create Position Attribute=true
    Create Sprite Rotation Attribute=true
    Create SpriteSize Attribute=true
    Position:Vector3f=&Add_Vector(
     A=&Add_Vector(A=Linked() {linkedVariable {name=Particles.Position, type=NiagaraPosition}}, B=$Particles.Initial.RandomOffset_02)
     B=&RotateVectorByQuaternion(Vector To Rotate=$Particles.OrientationOffset, Quaternion=$Particles.MeshOrientation)
    )
    SpriteRotation:NiagaraFloat=&RandomRangeFloat(Minimum=0, Maximum=1)
    SpriteSize=&Vector2DFromFloat(Value=&FloatFromCurve(
      FloatCurve=DataInterface() {propertyValues {Curve {keys [{arriveTangent=6.1188, interpMode=RCIM_Cubic, leaveTangent=6.1188, tangentMode=RCTM_User, time=0, value=1}, <at 0.9949 3.019>]}}}
      Scale Curve=&RandomRangeFloat(Minimum=10, Maximum=30)
     ))
    Write to Module Namespace=true
   ) {module=FN_Sprite_v3}
   /SetVariables(Particles.Black Smoke Color:LinearColor=&MakeLinearColorFromVectorAndFloat(
     "Vector (RGB)"=<0 0 0>
     "Float (Alpha)"=&FloatFromCurve(
      FloatCurve=DataInterface() {propertyValues {Curve {keys [<at -0.0044 -0.015 RCIM_Cubic>, {interpMode=RCIM_Cubic, leaveTangent=0.0149, time=0.3249, value=0.9916}, <at 0.9 0>]}}}
      Scale Curve=8
     )
    )) {isSetParameters=true}
  ]
  rendererClasses [&NiagaraMeshRendererProperties, &NiagaraSpriteRendererProperties]
  renderers [
   {rendererClass=&NiagaraMeshRendererProperties, rendererIndex=0}
   {rendererClass=&NiagaraSpriteRendererProperties, rendererIndex=1}
   {rendererClass=&NiagaraSpriteRendererProperties, rendererIndex=2}
  ]
 }
 Smoke_Ribbons {
  emitterUpdateScript [
   &EmitterState()
   &FN_SpawnRibbon_v2(
    Number of Ribbons=HlslExpression() {hlslExpression="Engine.QualityLevel > 2 ? 4 : 2"}
    Spawn Mode=::NewEnumerator0
    SpawnRate=HlslExpression() {hlslExpression="Engine.QualityLevel > 2 ? 30 : 20"}
   )
   &FN_Integrator() {enabled=false}
  ]
  particleSpawnScript [
   /Game/Effects/Niagara/Modules/Ribbons/FN_InitializeRibbon_v4(
    Custom Offset=true
    Enable Skeletal Position=true
    Lifetime:NiagaraFloat=&Min_Float(A:NiagaraFloat=$Emitter.Age, B:NiagaraFloat=0.6)
    Local Vector Offset:Vector3f=&TransformVector(
     Vector=&FN_FrontEndVector3Modifier(Gameplay=<0 0 1>, Front End=<0 0 0>)
     Source Space=::World
     Destination Space=::Local
    )
    Offset Amount=-10
    Ribbon Width=20
    SkeletalMesh:NiagaraDataInterfaceSkeletalMesh=$User.SkeletalMesh
   )
   &FN_FlipRibbonV(Offset=1)
   &AddVelocity(Velocity=&FN_FrontEndVector3Modifier(
     Gameplay=&Add_Vector(
      A=&TransformVector(
       Vector=&Multiply_VectorByFloat(
        Vector=&NormalizeVector(Vector To Normalize=$Engine.Owner.Velocity)
        Float=&Multiply_Float(A=400, B=HlslExpression() {hlslExpression="sign(dot(Engine.Owner.Velocity, float3(0,0,1)))"})
       )
       Source Space=::World
       Destination Space=::Local
      )
      B=&self:ScratchDynamicInput(
       Multiplier:NiagaraFloat=&Lerp_Float(
        A=0
        B=&SkydiveHandsFeetFloatModifier(Hands:NiagaraFloat=200, Feet:NiagaraFloat=50)
        Alpha=$System.Dive
       )
       Source:Vector3f=Linked() {linkedVariable {name=Particles.Position, type=NiagaraPosition}}
      )
     )
     Front End=&RotateVectorByQuaternion(
      Vector To Rotate=&SkydiveHandsFeetVector3Modifier(
       Hands:Vector3f=&MakeVector(
        X=&Multiply_Float(A=$User.ParaGlideLeanAlpha, B=HlslExpression() {hlslExpression="Particles.SpawnGroup && 1 ? -80 : 80"})
        Y=0
        Z=300
       )
       Feet:Vector3f=<0 0 300>
      )
      Quaternion=&MakeQuatFromAxisAngle(Axis=<1 0 0>, AngleInDegrees=&Multiply_Float(A=$User.ParaGlideLeanAlpha, B=63.4543))
     )
    ))
   &self:ScratchModule() {enabled=false, module=ScratchModule}
   &AddVelocity(Scale Added Velocity=<10 10 10>, Velocity=<0 0 50>) {enabled=false, module=AddVelocity001}
  ]
  particleUpdateScript [
   &SolveForcesAndVelocity(Clamp Velocity=false, Limit Acceleration=false)
   /Game/Effects/Niagara/Modules/Ribbons/FN_UpdateRibbon_v5(Enable Skeletal Position=true, SkeletalMesh:NiagaraDataInterfaceSkeletalMesh=$User.SkeletalMesh)
   /SetVariables(Particles.HandsFeetWidthModifier:NiagaraFloat=&SkydiveHandsFeetFloatModifier(Hands:NiagaraFloat=&Lerp_Float(A=0.4, B=1, Alpha=$System.FN_Integrator001.Value), Feet:NiagaraFloat=0.4)) {isSetParameters=true}
   &FN_RibbonMaterial_v2(
    Color:LinearColor=&MultiplyLinearColor(
     Linear Color=&LerpLinearColors(StartColor=<rgba 0.5 0 0 1>, EndColor=<rgba 1 0 0 1>, LerpFactor=$System.FN_Integrator001.Value)
     Scale Factor=&MakeVector4(
      X=&FloatFromCurve(FloatCurve=DataInterface() {propertyValues {Curve {keys [
           {arriveTangent=-1.9982, interpMode=RCIM_Cubic, leaveTangent=-1.9982, tangentMode=RCTM_User, time=0.1232, value=0.9856}
           <at 0.3605 0 RCIM_Cubic>
          ]}}})
      Y=0
      Z=0
      W=1
     )
    )
    Erosion Amount:NiagaraFloat=&FloatFromCurve(FloatCurve=DataInterface() {propertyValues {Curve {keys [
         {arriveTangent=20.0838, interpMode=RCIM_Cubic, leaveTangent=20.0838, tangentMode=RCTM_User, time=0.457, value=-0.0134}
         {arriveTangent=4.1923, interpMode=RCIM_Cubic, leaveTangent=4.1923, time=0.5183, value=1.0651}
         <at 0.9838 2.1949>
        ]}}})
    Erosion Pan Offset U:NiagaraFloat=&RandomRangeFloat(Minimum=0.8171, Maximum=0.3325, Randomness Mode=::NewEnumerator1, Random Seed=$Particles.SpawnGroup, Fixed Random Seed=true)
    Erosion Pan Speed U=1.6441
    Erosion Power=3
    Erossion Scale U=2.8333
    Macro Distortion Amount:NiagaraFloat=&FloatFromCurve(FloatCurve=DataInterface() {propertyValues {Curve {keys [<at 0 0>,<at 1 1>]}}}, Scale Curve=0.02)
    Macro Pan Speed=-0.7
    Macro Scale U=2.8704
    Opacity Erosion:NiagaraFloat=&FloatFromCurve(FloatCurve=DataInterface() {propertyValues {Curve {keys [<at 0.4476 -0.0139>,<at 0.8038 1.3641>]}}})
    Opacity Pan U Offset:NiagaraFloat=&RandomRangeFloat(Minimum=-0.5818, Maximum=1, Randomness Mode=::NewEnumerator1, Random Seed=$Particles.SpawnGroup, Fixed Random Seed=true)
    Opacity Pan U Speed=0.1818
    Opacity Power=0.6091
    Opacity Scale U:NiagaraFloat=&RandomRangeFloat(Minimum=0.8171, Maximum=0.3325, Randomness Mode=::NewEnumerator1, Random Seed=$Particles.SpawnGroup, Fixed Random Seed=true)
    TIme:NiagaraFloat=$System.FN_Integrator.Value
    UV0 Scale:NiagaraFloat=&Divide_Float(A=&Min_Float(A:NiagaraFloat=$Emitter.Age, B:NiagaraFloat=3), B=3)
   )
   &self:FN_Ribbon(RibbonWidth:NiagaraFloat=&Lerp_Float(A=5, B=20, Alpha=$Particles.NormalizedAge)) {module=FN_Ribbon002}
   &FN_RibbonMaterial_v2(
    Color:LinearColor=&MakeLinearColorFromVectorAndFloat(
     "Vector (RGB)"=&MakeVector(
      X=&FloatFromCurve(FloatCurve=DataInterface() {propertyValues {Curve {keys [<at 0 1>,<at 0.5561 -0.0042>]}}})
      Y=0
      Z=0
     )
     "Float (Alpha)"=2
    )
    Macro Distortion Amount=0.02
    Macro Pan Speed=0.6091
    Opacity Erosion:NiagaraFloat=&FloatFromCurve(FloatCurve=DataInterface() {propertyValues {Curve {keys [<at -0.0029 0.7857>,<at 0.185 0>,<at 0.7498 0>,<at 0.9912 0.991>]}}})
    Opacity Pan U Offset:NiagaraFloat=&RandomRangeFloat(Minimum=-1.4096, Maximum=12.2483, Randomness Mode=::NewEnumerator1, Random Seed=$Particles.SpawnGroup, Fixed Random Seed=true)
    Opacity Pan U Speed=0.4818
    Opacity Power:NiagaraFloat=&FloatFromCurve(FloatCurve=DataInterface() {propertyValues {Curve {keys [<at 0.0062 1.262>,<at 0.2588 0.3977>,<at 1 1.3237>]}}})
    Opacity Scale U=0.7636
    Scale Alpha:NiagaraFloat=&FloatFromCurve(FloatCurve=DataInterface() {propertyValues {Curve {keys [<at 0.5015 0.9473>,<at 0.7569 0.0123>]}}})
    Scale Color:Vector3f=&VectorFromCurve(VectorCurve=DataInterface() {propertyValues {XCurve {keys [<at 0 1>,<at 0.4378 0.0007>]}, YCurve {keys [<at 0 1>,<at 1 0>]}, ZCurve {keys [<at 0 1>,<at 1 0>]}}})
    TIme:NiagaraFloat=$System.FN_Integrator.Value
    UV0 Scale:NiagaraFloat=&Divide_Float(A=&Min_Float(A:NiagaraFloat=$Emitter.Age, B:NiagaraFloat=3), B=3)
    Write to Module Namespace=true
   ) {module=FN_RibbonMaterial_v2001}
   &self:FN_Ribbon(
    RibbonWidth:NiagaraFloat=&Multiply_Float(A=&Lerp_Float(A=15, B=20, Alpha=$Particles.NormalizedAge), B=$Particles.HandsFeetWidthModifier)
    Write to Moudle Namespace=true
   ) {module=FN_Ribbon}
   &FN_RibbonMaterial_v2(
    Color:LinearColor=&MakeLinearColorFromVectorAndFloat(
     "Vector (RGB)"=<0 0 0>
     "Float (Alpha)"=&FloatFromCurve(
      FloatCurve=DataInterface() {propertyValues {Curve {keys [
          {arriveTangent=0.0379, interpMode=RCIM_Cubic, leaveTangent=0.0379, tangentMode=RCTM_User, time=0.0033, value=0.9565}
          {arriveTangent=-1.2407, interpMode=RCIM_Cubic, leaveTangent=-1.2407, time=0.4081, value=0.9743}
          <at 0.7743 0>
         ]}}}
      Scale Curve=/Game/Effects/Niagara/DynamicInputs/Math/FN_Fit(
       Clamp:NiagaraBool=true
       Value:NiagaraFloat=$Emitter.Age
       SourceMin:NiagaraFloat=0.4
       SourceMax:NiagaraFloat=2
       DestMin:NiagaraFloat=0
       DestMax:NiagaraFloat=1
      )
     )
    )
    Macro Pan Speed=0.6091
    Opacity Erosion:NiagaraFloat=&FloatFromCurve(FloatCurve=DataInterface() {propertyValues {Curve {keys [
         <at -0.0159 0.0033>
         {arriveTangent=0.1362, interpMode=RCIM_Cubic, leaveTangent=0.1362, time=0.4038, value=0.0166}
         {arriveTangent=1.6585, interpMode=RCIM_Cubic, leaveTangent=1.6585, time=0.7161, value=0.1031}
         <at 0.9912 0.991>
        ]}}})
    Opacity Pan U Offset:NiagaraFloat=&RandomRangeFloat(Minimum=-1.4096, Maximum=4.8313, Randomness Mode=::NewEnumerator1, Random Seed=$Particles.SpawnGroup, Fixed Random Seed=true)
    Opacity Pan U Speed=0.3545
    Opacity Power:NiagaraFloat=&FloatFromCurve(FloatCurve=DataInterface() {propertyValues {Curve {keys [
         <at -0.0011 0.5398 RCIM_Cubic>
         {arriveTangent=3.4118, interpMode=RCIM_Cubic, leaveTangent=3.4118, time=0.2712, value=0.5733}
         <at 1.0241 4.0375>
        ]}}})
    Opacity Scale U=0.9273
    Scale Color:Vector3f=&VectorFromCurve(VectorCurve=DataInterface() {propertyValues {XCurve {keys [<at 0 0.2>,<at 0.6204 0.0029>]}, YCurve {keys [<at 0 0.2>,<at 1 0>]}, ZCurve {keys [<at 0 0.2>,<at 1 0>]}}})
    TIme:NiagaraFloat=$System.FN_Integrator.Value
    UV0 Scale:NiagaraFloat=&Divide_Float(A=&Min_Float(A:NiagaraFloat=$Emitter.Age, B:NiagaraFloat=3), B=3)
    Write to Module Namespace=true
   ) {module=FN_RibbonMaterial_v2002}
   &self:FN_Ribbon(
    RibbonWidth:NiagaraFloat=&Multiply_Float(A=&Lerp_Float(A=30, B=60, Alpha=$Particles.NormalizedAge), B=$Particles.HandsFeetWidthModifier)
    Write to Moudle Namespace=true
   ) {module=FN_Ribbon001}
   &FN_RibbonMaterial_v2(
    Color Pan U Speed=0.5364
    Color Scale U=0.9441
    Erosion Amount:NiagaraFloat=&FloatFromCurve(FloatCurve=DataInterface() {propertyValues {Curve {keys [<at 0.1769 0.9222 RCIM_Cubic>,<at 0.522 2.3845 RCIM_Cubic>]}}})
    Erosion Pan Speed U=0.4727
    Erosion Power=0.2022
    Erossion Scale U=0.571
    Macro Distortion Amount:NiagaraFloat=&FloatFromCurve(FloatCurve=DataInterface() {propertyValues {Curve {keys [<at 0.2898 -0.0043>,<at 0.9846 0.1461>]}}}, Scale Curve=0.7818)
    Macro Pan Speed=0.6091
    Macro Scale U=0.5727
    Micro Distortion Amount:NiagaraFloat=&FloatFromCurve(FloatCurve=DataInterface() {propertyValues {Curve {keys [<at 0.1 0>,<at 0.4544 1.0291>]}}}, Scale Curve=0.0482)
    Micro Pan Speed=0.3
    Micro Scale U=0.2545
    Opacity Pan U Speed=0.5364
    Opacity Scale U=0.9441
    Scale Color:Vector3f=&VectorFromCurve(VectorCurve=DataInterface() {propertyValues {XCurve {keys [<at 0 0.2>,<at 0.6204 0.0029>]}, YCurve {keys [<at 0 0.2>,<at 1 0>]}, ZCurve {keys [<at 0 0.2>,<at 1 0>]}}})
    TIme:NiagaraFloat=$System.FN_Integrator.Value
    Write to Module Namespace=true
   ) {module=FN_RibbonMaterial_v2003}
   &self:FN_Ribbon(
    RibbonWidth:NiagaraFloat=&Multiply_Float(A=&Lerp_Float(A=40, B=60, Alpha=$Particles.NormalizedAge), B=$Particles.HandsFeetWidthModifier)
    Write to Moudle Namespace=true
   ) {module=FN_Ribbon003}
  ]
  rendererClasses [&NiagaraRibbonRendererProperties]
  renderers [
   {rendererClass=&NiagaraRibbonRendererProperties}
   {rendererClass=&NiagaraRibbonRendererProperties, rendererIndex=1}
   {rendererClass=&NiagaraRibbonRendererProperties, rendererIndex=2}
   {rendererClass=&NiagaraRibbonRendererProperties, rendererIndex=3}
   {rendererClass=&NiagaraRibbonRendererProperties, rendererIndex=4}
  ]
 }
 Spline_Meshes {
  simTarget=GPUComputeSim
  emitterUpdateScript [
   &EmitterState()
   &SpawnRate(
    Spawn Probability=&FN_FrontEndFloatModifier(Gameplay=$System.Dive, Front End=$User.ParaGlideLeanAlpha)
    SpawnRate=20
    Use Spawn Probability=true
   )
   &SpawnRate(Use Spawn Probability=false) {enabled=false, module=SpawnRate001}
   &SpawnBurst_Instantaneous(Spawn Count=1, Use Spawn Probability=false) {enabled=false}
  ]
  particleSpawnScript [
   /SetVariables(
    Particles.Flip:NiagaraFloat=&UniformAOrBFloat(A=-1)
    Particles.Height:NiagaraFloat=&RandomRangeFloat(Minimum=1, Maximum=1.5)
    Particles.Radius:NiagaraFloat=&RandomRangeFloat(Minimum=0.5, Maximum=1)
   ) {isSetParameters=true}
   &InitializeParticle(
    Color=<rgba 1 0 0 1>
    Lifetime=&Multiply_Float(
     A=&RandomRangeFloat(Minimum=1.5, Maximum=2.5)
     B=&FN_FrontEndFloatModifier(Gameplay=1, Front End=0.5)
    )
    Mesh Scale=&Multiply_Vector(
     A=&MakeVector(X=$Particles.Radius, Y=$Particles.Radius, Z=$Particles.Height)
     B=&MakeVector(X=$Particles.Flip, Y=1, Z=1)
    )
    Module.Write Color=true
    Module.Write Lifetime=true
    Module.Write Mass=true
    Module.Write Material Random=true
    Module.Write Position=false
    Module.Write Scale=true
    Module.Write SpriteRotation=true
    Module.Write SpriteSize=true
    Sprite Rotation=&RandomRangeFloat(Minimum=0, Maximum=1)
    Sprite Size=&Multiply_Vector2(
     A=<20 50>
     B=&Vector2DFromFloat(Value=&RandomRangeFloat(Minimum=1, Maximum=3))
    )
   )
   &FN_SetSpawnGroup_v3(Number of Groups=2, Pattern:ENiagaraPartitionWaveForm=::NewEnumerator1)
   &SkeletalMeshLocation(
    Bone / Socket Index=$Particles.SpawnGroup
    Bone Sampling Mode=::NewEnumerator4
    Inherit Velocity=false
    Interpolate Mesh Sampling=false
    MaskSpawnGroup=false
    Skeletal Mesh=$User.SkeletalMesh
   )
   &AddVelocity(
    Scale Added Velocity=<0.5 0.5 0.5>
    Velocity=&FN_FrontEndVector3Modifier(
     Gameplay=&Add_Vector(
      A=&TransformVector(
       Vector=&Multiply_VectorByFloat(
        Vector=&NormalizeVector(Vector To Normalize=$Engine.Owner.Velocity)
        Float=&Multiply_Float(A=200, B=HlslExpression() {hlslExpression="sign(dot(Engine.Owner.Velocity, float3(0,0,1)))"})
       )
       Source Space=::World
       Destination Space=::Local
      )
      B=&self:ScratchDynamicInput(
       Multiplier:NiagaraFloat=&Lerp_Float(
        A=0
        B=&SkydiveHandsFeetFloatModifier(Hands:NiagaraFloat=50, Feet:NiagaraFloat=25)
        Alpha=$System.Dive
       )
       Source:Vector3f=Linked() {linkedVariable {name=Particles.Position, type=NiagaraPosition}}
      )
     )
     Front End=&RotateVectorByQuaternion(
      Vector To Rotate=&SkydiveHandsFeetVector3Modifier(
       Hands:Vector3f=&MakeVector(
        X=&Multiply_Float(A=$User.ParaGlideLeanAlpha, B=HlslExpression() {hlslExpression="Particles.SpawnGroup && 1 ? -40 : 40"})
        Y=0
        Z=150
       )
       Feet:Vector3f=<0 0 150>
      )
      Quaternion=&MakeQuatFromAxisAngle(Axis=<1 0 0>, AngleInDegrees=&Multiply_Float(A=$User.ParaGlideLeanAlpha, B=63.4543))
     )
    )
   )
   &InitialMeshOrientation(
    Orientation Vector=<0 0 1>
    Rotation=&MakeVector(X=0, Y=0, Z=&RandomRangeFloat(Minimum=-1, Maximum=1))
    Use Orientation Vector=false
    Use Rotation Vector=true
   )
   &AddVelocity(Velocity=<0 0 10>) {enabled=false, module=AddVelocity001}
  ]
  particleUpdateScript [
   &ParticleState()
   &SolveForcesAndVelocity(Clamp Velocity=false, Limit Acceleration=false)
   &ScaleMeshSize(Scale Factor=&VectorFromFloat(Value=&FloatFromCurve(FloatCurve=DataInterface() {propertyValues {Curve {keys [{arriveTangent=1.7682, interpMode=RCIM_Cubic, leaveTangent=1.7682, tangentMode=RCTM_User, time=0.0011, value=0.9204}, <at 1.0138 1.9699>]}}})))
   /SetVariables(Particles.MeshOrientation=&MultiplyQuaternion(
     Quaternion A=&FindQuatBetweenTwoVectors(From Vector=<0 0 1>, To Vector=$Particles.Velocity)
     Quaternion B=$Particles.Initial.MeshOrientation
    )) {isSetParameters=true, module=SetVariables_#2}
   &DynamicMaterialParameters(
    Index 0 Param 1=&FloatFromCurve(FloatCurve=DataInterface() {propertyValues {Curve {keys [
         {arriveTangent=6.1791, interpMode=RCIM_Cubic, leaveTangent=6.1791, tangentMode=RCTM_User, time=0, value=0}
         {interpMode=RCIM_Cubic, tangentMode=RCTM_User, time=0.1943, value=1}
         {arriveTangent=-0.178, interpMode=RCIM_Cubic, leaveTangent=-0.178, tangentMode=RCTM_User, time=0.75, value=0.9}
         <at 1 0 RCIM_Cubic>
        ]}}})
    Index 0 Param 2=&FloatFromCurve(FloatCurve=DataInterface() {propertyValues {Curve {keys [
         {arriveTangent=9.0034, interpMode=RCIM_Cubic, leaveTangent=9.0034, tangentMode=RCTM_User, time=0, value=0}
         {interpMode=RCIM_Cubic, tangentMode=RCTM_User, time=0.2, value=1}
         {interpMode=RCIM_Cubic, tangentMode=RCTM_User, time=0.75, value=1}
         {arriveTangent=-7.1709, interpMode=RCIM_Cubic, leaveTangent=-7.1709, tangentMode=RCTM_User, time=1, value=0}
        ]}}})
    Index 0 Param 4=&RandomRangeFloat(Minimum=0, Maximum=1, Randomness Mode=::NewEnumerator1, Fixed Random Seed=true)
    Index 1 Param 1=&Multiply_Float(A=$Particles.NormalizedAge, B=1)
    Index 1 Param 3=&RandomRangeFloat(Minimum=2, Maximum=3, Randomness Mode=::NewEnumerator1, Fixed Random Seed=true)
    Module.Param0WriteEnabled=true
    Module.Param0WriteEnabled1=true
    Module.Param1WriteEnabled=true
    Module.Param1WriteEnabled1=true
    Module.Param2WriteEnabled=false
    Module.Param2WriteEnabled1=true
    Module.Param3WriteEnabled=true
    Module.Param3WriteEnabled1=true
    Write Parameter Index 1=true
   )
   &Color(
    Color=<rgba 0 0 0 1>
    Scale Alpha=&FloatFromCurve(FloatCurve=DataInterface() {propertyValues {Curve {keys [<at 0 1>,<at 0.6562 0.0123>]}}})
   )
   &FN_SpriteUV(
    Randomize Sprite UV scale=true
    Seed=3
    Sub UV Play Rate:NiagaraFloat=&RandomRangeFloat(Minimum=0.5, Maximum=1)
    SubUV Animation Mode:ENiagara_SubUVLookupMode=::NewEnumerator3
    SubUV Play Rate Time:NiagaraFloat=$Particles.NormalizedAge
    SubUV Start Time Offset:NiagaraFloat=&RandomRangeFloat(Minimum=0, Maximum=1, Randomness Mode=::NewEnumerator1)
   ) {module=FN_SpriteUV002}
   /NiagaraFN/Sprite/FN_SpriteAlignmentAndFacing(SpriteAlignment:Vector3f=&RotateVectorByQuaternion(Vector To Rotate=<0 0 1>, Quaternion=$Particles.MeshOrientation))
   &FN_Sprite_Attributes(
    Create SpriteSize Attribute=true
    SpriteSize=&Multiply_Vector2(
     A=$Particles.Initial.SpriteSize
     B=&Vector2DFromCurve(
      Vector2Curve=DataInterface() {propertyValues {
        XCurve {keys [{arriveTangent=0.9003, interpMode=RCIM_Cubic, leaveTangent=0.9003, tangentMode=RCTM_User, time=0, value=1}, <at 0.9728 1.3697 RCIM_Cubic>]}
        YCurve {keys [{arriveTangent=2.8308, interpMode=RCIM_Cubic, leaveTangent=1.0055, tangentMode=RCTM_User, time=0, value=1}, <at 0.9903 2.0149 RCIM_Cubic>]}
       }}
      ScaleCurve=<1 1>
     )
    )
   ) {module=FN_Sprite}
  ]
  rendererClasses [&NiagaraMeshRendererProperties, &NiagaraSpriteRendererProperties]
  renderers [
   {rendererClass=&NiagaraMeshRendererProperties, rendererIndex=0}
   {rendererClass=&NiagaraMeshRendererProperties, rendererIndex=1}
   {rendererClass=&NiagaraSpriteRendererProperties, rendererIndex=2}
  ]
 }
}
summary {
 systemName=NS_Contrail_FutureSamurai_02
 emitters [
  {emitterName=Smoke_Ribbons, rendererClasses [&NiagaraRibbonRendererProperties]}
  {emitterName=Hand_Smoke, rendererClasses [&NiagaraSpriteRendererProperties]}
  {emitterName=Skulls, rendererClasses [&NiagaraMeshRendererProperties, &NiagaraSpriteRendererProperties]}
  {
   emitterName=FN_AttachedRibbon_v3001
   rendererClasses [&NiagaraRibbonRendererProperties, &NiagaraSpriteRendererProperties]
  }
  {
   emitterName=Spline_Meshes
   simTarget=GPUComputeSim
   rendererClasses [&NiagaraMeshRendererProperties, &NiagaraSpriteRendererProperties]
  }
 ]
 userVariables [
  User.bIsFrontEnd {type=NiagaraBool}
  User.bIsFrontEndPreview {type=NiagaraBool}
  User.ParaGlideLeanAlpha {type=NiagaraFloat, defaultValue=0, description=""}
  User.PlayerSpeed {type=NiagaraFloat, defaultValue=0, description=""}
  User.RibbonIDOffset {type=NiagaraInt32}
  User.SkeletalMesh {
   type=NiagaraDataInterfaceSkeletalMesh
   defaultValue {
    struct=NiagaraExt_VariableValue_DataInterface
    value {dataInterface=&self:NiagaraDataInterfaceSkeletalMesh_0}
   }
  }
 ]
}
